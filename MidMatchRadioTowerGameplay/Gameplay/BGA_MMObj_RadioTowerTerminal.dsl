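; Radio tower terminal for the mid-match objective, where a player swipes an access card to secure a storm forecast reward for the squad.
; On the server, success consumes the card, applies a GameplayEffect to squadmates, fires quest and objective events, and captures the tower.
; Without a card, a replicated error screen shows and hotfix-gated logic alerts guards and pings the boss.
; Tower-driven states update the screen child actor, audio and markers, with distance-gated progress ticks.
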(event EventBeginPlay
  ;[0.0453 0.1615 0.05378] BeginPlay: Cache the screen child actor.
  (|CallParentFunction)
  (Networking|CanExecuteCosmeticEvents
    (:True
      (bind _screen_childactor (Variables|Default|GetScreen_ChildActor)
        (:then
          (bind _asb_mmobj_radio_tower_terminal_screen (Utilities|Casting|CastToB_MMObj_RadioTowerTerminalScreen (Variables|/Script/Engine.ChildActorComponent|GetChildActor _screen_childactor)))
          (Variables|Transient|SetScreen _asb_mmobj_radio_tower_terminal_screen))
        (:else)))))

(event Interaction|EventBlueprintOnInteract
  (params (InteractingPawn :type Fort%20Pawn%20Object%20Reference) (InteractionBeingAttempted :type EInteractionBeingAttempted%20Enum :default "FirstInteraction"))
  (bind _returnvalue (Fortnite|Athena|GameState|GetGameStateAthena))
  (bind _returnvalue_2 (Actor|GetComponentbyClass _returnvalue "/MidMatchRadioTowerGameplay/Gameplay/B_MMObj_RadioTowerGameStateComponent.B_MMObj_RadioTowerGameStateComponent_C"))
  (bind _radiotower (Tower|GetRadioTower))
  ;[0.1042 0.01687 0.09707] OInteract: Update to the Activated state and give reward GE to all players on the team.
  (|CallParentFunction InteractingPawn InteractionBeingAttempted)
  (Networking|SwitchHasAuthority
    (:Authority
      (bind _asfort_player_pawn_athena (Utilities|Casting|CastToFortPlayerPawnAthena InteractingPawn))
      (Fortnite|FortShippingLog self (Utilities|String|Append "OnInteract: PlayerPawn: " (Fortnite|GetPlayerNameSafe _asfort_player_pawn_athena)))
      (Animation|PlayPlayerSwipeAnimation _asfort_player_pawn_athena)
      (if (Reward|IsAcceptingKeycard)
        (if (Reward|DoesPlayerPawnHaveKeycard _asfort_player_pawn_athena)
          (Variables|Transient|SetbErrorScreen false)
          (Utilities|Time|ClearandInvalidateTimerbyHandle (Variables|Transient|GetErrorScreenDurationTimerHandle))
          (Fortnite|FortShippingLog self (Utilities|String|Append "OnInteract: Authorized access by player with keycard: " (Fortnite|GetPlayerNameSafe _asfort_player_pawn_athena)))
          (Reward|GrantRewardtoPlayerandTeam _asfort_player_pawn_athena)
          (Quest|SendQuestCompleteEvent _asfort_player_pawn_athena "(TagName=\"CustomVerb.RadioTower.Activated\")") ; Notify radio tower successful interact quest completion
          (Ability|ExecuteGameplayCueReplicated (Variables|Settings|GetGC_Success))
          (Networking|FlushNetDormancy)
          (Utilities|FlowControl|Sequence
            (:then_0
              (Utilities|IsValid _radiotower
                (:Is%20Valid
                  (Public|SetRadioTowerState _radiotower "NewEnumerator7"))))
            (:then_1
              (Utilities|IsValid _returnvalue
                (:Is%20Valid
                  (Utilities|IsValid _returnvalue_2
                    (:Is%20Valid
                      (Objective|NotifyRadioTowerObjectiveComplete _returnvalue_2 _asfort_player_pawn_athena)))))))
          (else
            (Variables|Transient|SetbErrorScreen true)
            (bind _returnvalue_3 (Utilities|Time|SetTimerbyEvent (EventDispatchers|CreateEvent :!event HandleErrorScreenTimerComplete) (ScalableFloat|GetValueatLevel (Variables|Hotfix|GetHotfix_TerminalErrorScreenDuration))))
            (Variables|Transient|SetErrorScreenDurationTimerHandle _returnvalue_3)
            (Ability|ExecuteGameplayCueReplicated (Variables|Settings|GetGC_Fail))
            (Networking|FlushNetDormancy)
            (Guards|AlertGuards _asfort_player_pawn_athena)
            (Guards|MarkGuardforPlayer _asfort_player_pawn_athena)))
        (else
          ;;[1 0 0] Warning: Interaction happening in an invalid state.
          (Fortnite|FortShippingLog self (Utilities|String|Append "OnInteract: Player interacting with terminal when the terminal is not accepting keycard: " (Fortnite|GetPlayerNameSafe _asfort_player_pawn_athena)) true))))))

(event Custom|HandleTerminalStateUpdate
  (params (TerminalState :type MMObjRadioTowerState%20Enum :default "NewEnumerator5"))
  (bind _stateupdateburstsoundmap (Variables|Sound|GetStateUpdateBurstSoundMap))
  (bind _value (Utilities|Map|Find _stateupdateburstsoundmap TerminalState))
  ;[0.03646 0.03406 0.007275] HandleTerminalStateUpdate: Used for burst effects.
  ; Called from B_MMObj_RadioTower directly.
  (Networking|CanExecuteCosmeticEvents
    (:True
      (bind _inrange (CallFunction|IsWithinPlayerRange))
      (if _inrange
        (Utilities|IsValid _value
          (:Is%20Valid
            (Audio|PlaySoundatLocation _value (Transformation|GetWorldLocation (Variables|Default|GetAmbientLoopAudioComponent)))))))))

(fn AlertGuards
  (params (InteractingPlayerPawn :type Fort%20Player%20Pawn%20Object%20Reference))
  ;[0.01795 0.1167 0.1406] AlertGuards: Alert the guards about the interacting player pawn so that they go investigate the terminal.
  (Networking|SwitchHasAuthority
    (:Authority
      (Utilities|IsValid InteractingPlayerPawn
        (:Is%20Valid
          (if (ScalableFloat|GetValueasBool (Variables|Hotfix|GetHotfix_FailInteractAlertGuards))
            (bind _guardspawner (Guards|GetGuardSpawner))
            (Utilities|IsValid _guardspawner
              (:Is%20Valid
                (bind _guardpawns (Helpers|GetAllGuardPawns _guardspawner))
                (for _array_element _guardpawns
                  (Utilities|IsValid _array_element
                    (:Is%20Valid
                      (bind _asfort_athena_aibot_controller (Utilities|Casting|CastToFortAthenaAIBotController (Pawn|GetController _array_element)))
                      (Class|FortAthenaAIBotController|QueueStim _asfort_athena_aibot_controller InteractingPlayerPawn (Transformation|GetActorLocation InteractingPlayerPawn) (Math|Vector|GetUnitDirection(Vector) (Transformation|GetActorLocation _array_element) (Transformation|GetActorLocation InteractingPlayerPawn)) "Seeing" 1.0 "None"))))))))))))

(fn BlueprintCanInteract
  (params (InteractingPawn :type Fort%20Pawn%20Object%20Reference) (InteractionBeingAttempted :type EInteractionBeingAttempted%20Enum :default "FirstInteraction") (InteractionType :type TInteractionType%20Enum :default "IT_NoInteraction"))
  ;[0.0231 0.1222 0.1875] Can Interact when in the Interactive state.
  (return (Reward|IsAcceptingKeycard)))

(fn BlueprintGetAllHighlightableComponents
  ;[0.0433 0.09931 0.151] Highlight the static mesh component.
  (return (Utilities|Array|MakeArray (Variables|Default|GetStaticMeshComponent)) true))

(fn BlueprintGetInteractionString
  (params (InteractingPawn :type Fort%20Pawn%20Object%20Reference) (InteractionBeingAttempted :type EInteractionBeingAttempted%20Enum :default "FirstInteraction"))
  (bind _returnvalue (Fortnite|GetFortPlayerControllerfromActor InteractingPawn))
  ;[0.01636 0.09989 0.125] GetInteractionString: Returns the interaction string shown to the player.
  (Utilities|FlowControl|Sequence
    (:then_0
      (Utilities|IsValid InteractingPawn
        (:Is%20Valid
          (Utilities|IsValid _returnvalue
            (:Is%20Valid
              (Utilities|IsValid (Inventory|FindExistingItemForDefinition :self _returnvalue :ItemDefinition (Variables|Settings|GetKeyItem) :bInStorageVault false)
                (:Is%20Valid
                  (return "NSLOCTEXT(\"[CD98012BEFCC1182438DD7CDBAB2CE87]\", \"E3C293924FFCF27659B23EA379FE8773\", \"Secure Forecast\")"))
                (:Is%20Not%20Valid
                  (return "NSLOCTEXT(\"[CD98012BEFCC1182438DD7CDBAB2CE87]\", \"8F3333C94DDABCDFF948D7BC4F13299C\", \"Requires Access Card\")"))))))))
    (:then_1
      (return ""))))

(fn BlueprintGetInteractionTime
  (params (InteractingPawn :type Fort%20Pawn%20Object%20Reference) (InteractionBeingAttempted :type EInteractionBeingAttempted%20Enum :default "FirstInteraction"))
  ;[0.09375 0.0322 0.08166] Update the interaction duration.
  (return (Variables|Settings|GetInteractionDuration) true))

(fn DoesPlayerPawnHaveKeycard
  (params (PlayerPawn :type Fort%20Player%20Pawn%20Athena%20Object%20Reference))
  (bind _returnvalue (Fortnite|GetFortPlayerControllerfromActor PlayerPawn))
  ;[0.03811 0.1593 0.1875] Returns if the player pawn has the keycard in their inventory.
  (Utilities|FlowControl|Sequence
    (:then_0
      (Utilities|IsValid _returnvalue
        (:Is%20Valid
          (Utilities|IsValid (Inventory|FindExistingItemForDefinition :self _returnvalue :ItemDefinition (Variables|Settings|GetKeyItem) :bInStorageVault false)
            (:Is%20Valid
              (return true))))))
    (:then_1
      (return false))))

(fn GetFocusedSocketLocation
  (return (Transformation|GetSocketLocation (Variables|Default|GetStaticMeshComponent) "Interact")))

(fn GetGuardSpawner
  (bind _returnvalue (WorldSubsystems|GetFortGlobalActorRegistrySubsystem))
  ;[0.007328 0.1233 0.1771] Returns the guard spawner from the world.
  (Utilities|FlowControl|Sequence
    (:then_0
      (Networking|SwitchHasAuthority
        (:Authority
          (Utilities|IsValid _returnvalue
            (:Is%20Valid
              (bind _outactors (Class|FortGlobalActorRegistrySubsystem|GetRegisteredActors _returnvalue (GameplayTags|MakeLiteralGameplayTag "(TagName=\"MMObj.RadioTower.GuardSpawner\")")))
              (for _array_element _outactors
                (bind _asb_mmobj_radio_tower_guard_spawner (Utilities|Casting|CastToB_MMObj_RadioTowerGuardSpawner _array_element))
                (if (== (Helpers|GetSpawnerIdTag _asb_mmobj_radio_tower_guard_spawner) (Variables|Instance|GetRadioTowerGuardSpawnerTag))
                  (return _asb_mmobj_radio_tower_guard_spawner))))))))
    (:then_1
      (return 0))))

(fn GetIconPlacement
  (params (SelfActor :type Actor%20Object%20Reference) (ViewingActor :type Actor%20Object%20Reference))
  (return (Transformation|GetSocketLocation (Variables|Default|GetStaticMeshComponent) "Interact") "0, 0, 0"))

(fn GetRadioTower
  (return (Variables|Instance|GetRadioTower)))

(fn GetRadioTowerState
  (bind _radiotower (Variables|Instance|GetRadioTower)
    (:then
      (return (Class|BMMObjRadioTower|GetRadioTowerState _radiotower)))
    (:else
      (return "NewEnumerator5"))))

(fn GrantRewardToPlayerAndTeam
  (params (PlayerPawn :type Fort%20Player%20Pawn%20Athena%20Object%20Reference))
  (bind _returnvalue (Fortnite|Athena|GameState|GetGameStateAthena))
  ;[0.0229 0.1206 0.1354] GrantRewardToPlayerAndTeam: Removes the key item from the player inventory and grants the reward to all players on the team.
  (Networking|SwitchHasAuthority
    (:Authority
      (Utilities|IsValid PlayerPawn
        (:Is%20Valid
          ;;[0.3002 0.03434 0.3021] Remove the key from the player's inventory.
          (bind _returnvalue_2 (Inventory|RemoveItemFromPlayer :PlayerController (Fortnite|GetFortPlayerControllerfromActor PlayerPawn) :ItemDefinition (Variables|Settings|GetKeyItem) :AmountToRemove 1 :bForceRemoval true :bSkipVerbRemoval false))
          (if (> _returnvalue_2 0)
            ;;[0.05049 0.224 0.04947] Grant reward GE to all players in the interacting player's team.
            (Utilities|IsValid _returnvalue
              (:Is%20Valid
                (bind _returnvalue_3 (Class|FortGameStateAthena|GetSquadMemberPlayerStates _returnvalue (Utilities|Casting|CastToFortPlayerStateAthena (Variables|/Script/Engine.Pawn|GetPlayerState PlayerPawn)) true))
                (for _array_element _returnvalue_3
                  (Utilities|IsValid _array_element
                    (:Is%20Valid
                      (Fortnite|FortShippingLog self (Utilities|String|Append "GrantRewardToPlayerAndTeam: Granting reward to player: " (Fortnite|GetPlayerNameSafe (PlayerState|GetCurrentPawn _array_element))))
                      (Utilities|IsValid (Ability|GetFortAbilitySystemComponent _array_element)
                        (:Is%20Valid
                          (GameplayEffects|ApplyGameplayEffectToSelf (Ability|GetFortAbilitySystemComponent _array_element) (Variables|Settings|GetStormScoutRewardGameplayEffectClass) 0.0 (GameplayEffects|MakeEffectContext (Ability|GetFortAbilitySystemComponent _array_element))))
                        (:Is%20Not%20Valid
                          _))
                      ;;[0.01643 0.04194 0.09896] Responsible for sending UI notification to the players that receive the reward.
                      (GameplayCue|ExecuteGameplayCue (PlayerState|GetCurrentPawn _array_element) (Variables|Settings|GetRadioTowerNotificationGameplayCueTag)))))))
            (else
              ;;[1 0 0] Warning: Failed to remove key item from player
              (Fortnite|FortShippingLog self (Utilities|String|Append "GrantRewardToPlayerAndTeam: Failed to remove key item from player: " (Fortnite|GetPlayerNameSafe PlayerPawn)) true))))))))

(fn HandleErrorScreenTimerComplete
  (Networking|SwitchHasAuthority
    (:Authority
      (Variables|Transient|SetbErrorScreen false)
      (Networking|FlushNetDormancy))))

(fn HandleRadioTowerStateChanged
  (params (NewState :type MMObjRadioTowerState%20Enum :default "NewEnumerator5"))
  ;[0.03591 0.2104 0.3281] HandleRadioTowerStateChanged:
  ; When the radio tower state changes to active, activate the terminal as well.
  ; When the radio tower state changes to activating, start activating the terminal as well.
  ; Called from the radio tower directly.
  (Utilities|FlowControl|Sequence
    (:then_0
      (Cosmetic|UpdateAudioCue)
      (MarkableActor|SetCanBeMarked (Utilities|Select :Index NewState :NewEnumerator5 false :NewEnumerator0 false :NewEnumerator6 true :NewEnumerator2 true :NewEnumerator7 false))
      (switch Utilities|FlowControl|Switch|SwitchonMMObjRadioTowerState NewState
        (:NewEnumerator0
          (States|HandleTerminalStateDeactive))
        (:NewEnumerator6
          (States|HandleTerminalStateActivating))
        (:NewEnumerator2
          (States|HandleTerminalStateActive))
        (:NewEnumerator7
          (States|HandleTerminalStateCaptured))))
    (:then_1
      (Default|CallOnTerminalStateChanged self NewState))))

(fn HandleTerminalState_Activating
  ;[0.01594 0.006093 0.05208] HandleTerminalState_Activating: Invoked when the state changes to activating.
  ; Called on both server and client.
  (Networking|CanExecuteCosmeticEvents
    (:True
      (Class|BMMObjRadioTowerTerminalScreen|SetEnableScreen (Variables|Transient|GetScreen) true)
      ;;[0.1094 0.0135 0.08618] Set a timer that updates the activation progress % on the screens.
      (bind _returnvalue (Utilities|Time|SetTimerbyEvent (EventDispatchers|CreateEvent :!event HandleUpdateActivationProgress) 1.0 true))
      (Variables|Transient|SetHandleUpdateActivationProgressTimerHandle _returnvalue)
      ;;[0.03228 0.2331 0.2917] Scanning storm.
      (Class|BMMObjRadioTowerTerminalScreen|UpdateScreen (Variables|Transient|GetScreen) "NewEnumerator6" (Variables|UI|GetActivatingText))
      (Class|BMMObjRadioTowerTerminalScreen|SetEnableSecondaryText (Variables|Transient|GetScreen) true)
      (States|HandleUpdateActivationProgress)
      (Class|BMMObjRadioTowerTerminalScreen|PlayActivateAnimation (Variables|Transient|GetScreen)))))

(fn HandleTerminalState_Active
  ;[0.1927 0.04173 0.1781] HandleTerminalState_Activated: Invoked when the state changes to Activated.
  ; Called on both server and client.
  (Networking|SwitchHasAuthority
    (:Remote
      (Class|BMMObjRadioTowerTerminalScreen|SetEnableScreen (Variables|Transient|GetScreen) true)
      (Class|BMMObjRadioTowerTerminalScreen|UpdateScreen (Variables|Transient|GetScreen) "NewEnumerator2" (Variables|UI|GetActiveText))
      (Utilities|Time|ClearandInvalidateTimerbyHandle (Variables|Transient|GetHandleUpdateActivationProgressTimerHandle))
      (Class|BMMObjRadioTowerTerminalScreen|SetEnableSecondaryText (Variables|Transient|GetScreen) false))))

(fn HandleTerminalState_Captured
  (bind _self self)
  ;[0.08955 0.04091 0.1771] HandleTerminalState_Ready: Invoked when the state changes to Interactive.
  ; Called on both server and client.
  (Networking|SwitchHasAuthority
    (:Authority
      (Map|SetEnableMapIcon false))
    (:Remote
      (Class|BMMObjRadioTowerTerminalScreen|SetEnableScreen (Variables|Transient|GetScreen) true)
      (Class|BMMObjRadioTowerTerminalScreen|UpdateScreen (Variables|Transient|GetScreen) "NewEnumerator7" (Variables|UI|GetCapturedText))
      (Class|BMMObjRadioTowerTerminalScreen|SetEnableSecondaryText (Variables|Transient|GetScreen) false)
      ;[0.02437 0.128 0.1562] Cancel the marker on this terminal.
      (for _array_element (Fortnite|GetLocalFortPlayerControllers true true)
        (Utilities|IsValid _array_element
          (:Is%20Valid
            (Utilities|IsValid (Actor|GetComponentbyClass _array_element "/Script/FortniteGame.AthenaMarkerComponent")
              (:Is%20Valid
                (Utilities|IsValid (MarkerComponent|FindMarkerbyMarkedActor (Actor|GetComponentbyClass _array_element "/Script/FortniteGame.AthenaMarkerComponent") _self)
                  (:Is%20Valid
                    (Class|FortWorldMarker|CancelMarker (MarkerComponent|FindMarkerbyMarkedActor (Actor|GetComponentbyClass _array_element "/Script/FortniteGame.AthenaMarkerComponent") _self))))))))))))

(fn HandleTerminalState_Deactive
  ;[1 0.6284 0.165] HandleTerminalState_Deactivated: Invoked when the state changes to Deactive.
  ; Called on both server and client.
  (Networking|SwitchHasAuthority
    (:Remote
      (bind _screen (Variables|Transient|GetScreen)
        (:then
          (Class|BMMObjRadioTowerTerminalScreen|SetEnableScreen _screen true)
          (Class|BMMObjRadioTowerTerminalScreen|UpdateScreen _screen "NewEnumerator0" (Variables|UI|GetDeactiveText)))
        (:else
          ;;[1 0 0] Warning: Screen is null.
          (Fortnite|FortShippingLog self "HandleTerminalState_Deactive: Screen is null." true))))))

(fn HandleUpdateActivationProgress
  (bind _returnvalue (Fortnite|Athena|GameState|GetGameStateAthena))
  (bind _returnvalue_2 (Actor|GetComponentbyClass _returnvalue "/MidMatchRadioTowerGameplay/Gameplay/B_MMObj_RadioTowerGameStateComponent.B_MMObj_RadioTowerGameStateComponent_C"))
  ;[0.1719 0.01794 0.1638] HandleUpdateActivationProgress: Update the activation progress on the screen based on the value from the game state component.
  ; Convert 0-1 float to 0-100%.
  (Networking|CanExecuteCosmeticEvents
    (:True
      (bind _inrange (CallFunction|IsWithinPlayerRange))
      (if _inrange
        (Utilities|IsValid _returnvalue
          (:Is%20Valid
            (Utilities|IsValid _returnvalue_2
              (:Is%20Valid
                (Class|BMMObjRadioTowerTerminalScreen|UpdateSecondaryText (Variables|Transient|GetScreen) (Utilities|Text|FormatText "INVTEXT(\"{X} %\")" (Math|Float|Truncate (* (Math|Float|MapRangeClamped (Math|Float|Clamp(Float) (+ (General|GetRadioTowerActivationProgress _returnvalue_2) (Variables|Settings|GetTowerActivationProgressOffset)) (Variables|Settings|GetTowerActivationProgressOffset)) (Variables|Settings|GetTowerActivationProgressOffset) 1.0 0.0 1.0) 1e+02))))
                (Audio|PlaySoundatLocation "/MidMatchRadioTowerGameplay/Sound/RadioTowerConsole_Tick_Cue.RadioTowerConsole_Tick_Cue" (Transformation|GetWorldLocation (Variables|Default|GetScreen_ChildActor)))))))))))

(fn IsAcceptingKeycard
  ;[0.1124 0.01864 0.1719] Keycard is accepted as long as the radio tower is in the Active State or Locked State.
  (return (Utilities|Enum|Equal(Enum) (Tower|GetRadioTowerState) "NewEnumerator2")))

(fn IsWithinPlayerRange
  (bind _returnvalue (Development|GetConsoleVariableFloatValue "Fort.RadioTowerTerminal.DistanceToUpdate"))
  (return (> _returnvalue (Transformation|GetSquaredDistanceTo (Game|GetPlayerPawn 0)))))

(fn MarkGuardForPlayer
  (params (InteractingPlayerPawn :type Fort%20Player%20Pawn%20Object%20Reference))
  (bind _returnvalue (Pawn|GetController InteractingPlayerPawn))
  (bind _returnvalue_2 (Actor|GetComponentbyClass _returnvalue "/Script/FortniteGame.AthenaMarkerComponent"))
  ;[0.08333 0.07332 0.02919] Mark the boss guard for the input player pawn.
  (Networking|SwitchHasAuthority
    (:Authority
      (Utilities|IsValid InteractingPlayerPawn
        (:Is%20Valid
          (Utilities|IsValid _returnvalue
            (:Is%20Valid
              (if (ScalableFloat|GetValueasBool (Variables|Hotfix|GetHotfix_FailInteractPingGuards))
                (bind _guardspawner (Guards|GetGuardSpawner))
                (Utilities|IsValid _guardspawner
                  (:Is%20Valid
                    (bind _bosspawn (Helpers|GetBossPawn _guardspawner))
                    (Utilities|IsValid _bosspawn
                      (:Is%20Valid
                        (Utilities|IsValid _returnvalue_2
                          (:Is%20Valid
                            (MarkerComponent|MarkActoronServer _returnvalue_2 _bosspawn false true)))))))))))))))

(fn OnRep_bErrorScreen
  ;[0.06382 0.02807 0.2552] Update the error state on the screen.
  (Networking|CanExecuteCosmeticEvents
    (:True
      (Class|BMMObjRadioTowerTerminalScreen|SetEnableErrorScreen (Variables|Transient|GetScreen) (Variables|Transient|GetbErrorScreen)))))

(fn PlayPlayerSwipeAnimation
  (params (PlayerPawn :type Fort%20Player%20Pawn%20Athena%20Object%20Reference))
  (bind _returnvalue (Ability|GetFortAbilitySystemComponent PlayerPawn))
  ;[0.07217 0.03157 0.09896] Play the player swipe animation using a gameplay effect.
  (Networking|SwitchHasAuthority
    (:Authority
      (Utilities|IsValid PlayerPawn
        (:Is%20Valid
          (Utilities|IsValid _returnvalue
            (:Is%20Valid
              (GameplayEffects|ApplyGameplayEffectToSelf _returnvalue (Variables|Settings|GetPawnSwipeAnimMontageGameplayEffect) 0.0 (GameplayEffects|MakeEffectContext _returnvalue)))
            (:Is%20Not%20Valid
              ;;[1 0 0] Warning: Invalid player pawn.
              (Fortnite|FortShippingLog self (Utilities|String|Append "PlayerPlayerSwipeAnimation: ASC is null for player: " (Fortnite|GetPlayerNameSafe PlayerPawn)) true))))
        (:Is%20Not%20Valid
          ;;[1 0 0] Warning: Invalid player pawn.
          (Fortnite|FortShippingLog self "PlayPlayerSwipeAnimation: Input player pawn is null." true))))))

(fn SendQuestCompleteEvent
  (params (PlayerPawn :type Fort%20Player%20Pawn%20Athena%20Object%20Reference) (QuestTag :type Gameplay%20Tag%20Structure))
  (bind _returnvalue (Fortnite|GetFortPlayerControllerfromActor PlayerPawn))
  ;[0.001906 0.08065 0.09375] SendQuestCompleteEvent: Let's the quest manager know that the specified quest has been completed by the input player pawn.
  (Networking|SwitchHasAuthority
    (:Authority
      (Utilities|IsValid PlayerPawn
        (:Is%20Valid
          (Utilities|IsValid _returnvalue
            (:Is%20Valid
              (CustomVerb|BroadcastCustomVerbMessage (CustomVerb|MakeCustomVerbMessage _returnvalue QuestTag))
              (Fortnite|FortShippingLog self (Utilities|String|Append "SendQuestCompleteEvent: Successfully broadcasted verb message " (GameplayTags|GetDebugStringfromGameplayTag QuestTag) " for player " (Fortnite|GetPlayerNameSafe PlayerPawn))))
            (:Is%20Not%20Valid
              ;;[1 0 0] Warning: PlayerController is null.
              (Fortnite|FortShippingLog self "SendQuestCompleteEvent: PlayerController is null." true))))
        (:Is%20Not%20Valid
          ;;[1 0 0] Warning: Input PlayerPawn is null.
          (Fortnite|FortShippingLog self "SendQuestCompleteEvent: Input PlayerPawn is null." true))))))

(fn SetEnableMapIcon
  (params (bEnable :type Boolean))
  ;[0.01375 0.2032 0.2448] Enable or disable the map icon.
  (Networking|SwitchHasAuthority
    (:Authority
      (bind _athenaspecialactor (Variables|Default|GetAthenaSpecialActor)
        (:then
          (if bEnable
            (Components|Activation|Activate _athenaspecialactor false)
            (else
              (Components|Activation|Deactivate _athenaspecialactor))))
        (:else)))))

(fn SpawnRadioTowerGuards
  ;[0.01876 0.02834 0.09896] SpawnRadioTowerGuards: Spawn the radio tower guards using the appropriate spawner.
  ; Called from B_MMObj_RadioTowerGameStateComponent.
  (Networking|SwitchHasAuthority
    (:Authority
      (bind _guardspawner (Guards|GetGuardSpawner))
      (Utilities|IsValid _guardspawner
        (:Is%20Valid
          (GuardSpawn|Public|RequestSpawnGuard _guardspawner self))))))

(fn UpdateAudioCue
  (bind _radiotowerstate (Tower|GetRadioTowerState))
  (bind _stateupdateloopsoundmap (Variables|Sound|GetStateUpdateLoopSoundMap))
  (bind _value (Utilities|Map|Find _stateupdateloopsoundmap _radiotowerstate))
  (bind _ambientloopaudiocomponent (Variables|Default|GetAmbientLoopAudioComponent))
  ;[0.003723 0.1727 0.2188] Update the audio loop component based on the state.
  (Networking|CanExecuteCosmeticEvents
    (:True
      (Utilities|IsValid _value
        (:Is%20Valid
          (Audio|Components|Audio|SetSound _ambientloopaudiocomponent _value)
          (Audio|Components|Audio|Play _ambientloopaudiocomponent 0.0))
        (:Is%20Not%20Valid
          (Audio|Components|Audio|FadeOut (Variables|Default|GetAmbientLoopAudioComponent) 0.5 0.0 "Linear" :!class /Script/Engine.AudioComponent))))))

(fn ConstructionScript)
