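; Mid-match objective radio tower actor that moves through deactive, activating, active and captured states.
; The server takes state from the game state component on phase advance, triggers guard spawns at the terminal, and locks state once captured.
; OnRep drives lights, radar signal, looping audio and the terminal, while a multicast plays burst sounds and clients track activation progress.
; Linked lights and terminal actors are located by box overlap when unset.
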(event EventBeginPlay
  ;[0.02681 0.151 0.02497] BeginPlay: Gets the state from B_MMObj_RadioTowerGameStateComponent.
  (|CallParentFunction)
  (Networking|SwitchHasAuthority
    (:Authority
      (Class|FortAsyncActionCurrentPlaylistReady|PlaylistDataReadyAsync
        (:OnReady
          (CallFunction|SyncLightsTransform)
          (CallFunction|UpdateRadioTowerState)
          (bind _returnvalue (Protected|GetRadioTowerStateComponent))
          (Default|BindEventtoOnRadioTowerPhaseAdvance _returnvalue (EventDispatchers|CreateEvent :!event OnTowerPhaseAdvance))
          (Default|BindEventtoOnRadioTowerSpawnGuards _returnvalue (EventDispatchers|CreateEvent :!event SpawnGuards))
          (bind _shouldspawntowerguards (Class|BMMObjRadioTowerGameStateComponent|GetRadioTowerShouldSpawnGuards _returnvalue))
          (if _shouldspawntowerguards
            (CallFunction|SpawnGuards)))))))

(event EventEndPlay
  (params (EndPlayReason :type EEndPlayReason%20Enum :default "Destroyed"))
  (bind _returnvalue (Fortnite|Athena|GameState|GetGameStateAthena))
  (bind _returnvalue_2 (Actor|GetComponentbyClass _returnvalue "/MidMatchRadioTowerGameplay/Gameplay/B_MMObj_RadioTowerGameStateComponent.B_MMObj_RadioTowerGameStateComponent_C"))
  ;[0.1354 0.08461 0.02039] EndPlay: Unbind delegates.
  (Utilities|FlowControl|Sequence
    (:then_0
      (Networking|SwitchHasAuthority
        (:Authority
          (bind _returnvalue_3 (Protected|GetRadioTowerStateComponent))
          (Default|UnbindEventfromOnRadioTowerPhaseAdvance _returnvalue_3 (EventDispatchers|CreateEvent :!event OnTowerPhaseAdvance))
          (Default|UnbindEventfromOnRadioTowerSpawnGuards _returnvalue_3 (EventDispatchers|CreateEvent :!event SpawnGuards)))))
    (:then_1
      (Networking|CanExecuteCosmeticEvents
        (:True
          (Utilities|IsValid _returnvalue
            (:Is%20Valid
              (Utilities|IsValid _returnvalue_2
                (:Is%20Valid
                  (Default|UnbindEventfromOnRadioTowerActivationProgressUpdated _returnvalue_2 (EventDispatchers|CreateEvent :!event HandleActivationProgressUpdated)))))))))
    (:then_2
      (|CallParentFunction EndPlayReason))))

(event Custom|Multicast%20HandleRadioTowerStateUpdate
  (params (NewState :type MMObjRadioTowerState%20Enum :default "NewEnumerator5"))
  (bind _stateupdateburstsoundmap (Variables|Settings|GetStateUpdateBurstSoundMap))
  (bind _value (Utilities|Map|Find _stateupdateburstsoundmap NewState))
  ;[0.1157 0.02506 0.1562] HandleRadioTowerStateUpdated: Useful to fire off burst effects.
  (Networking|CanExecuteCosmeticEvents
    (:True
      (Utilities|FlowControl|Sequence
        (:then_0
          (Utilities|IsValid _value
            (:Is%20Valid
              (Audio|PlaySoundatLocation _value (Transformation|GetWorldLocation (Variables|Default|GetAmbientLoopAudioComponent)))
              ;;[1 1 1] Close Sound
              (Audio|SpawnSoundAttached 0 (Variables|Default|GetAmbientLoopAudioCloseComponent)))))
        (:then_1
          (bind _radiotowerterminal (Variables|Instance|GetRadioTowerTerminal)
            (:then
              (Class|BGAMMObjRadioTowerTerminal|HandleTerminalStateUpdate _radiotowerterminal NewState))
            (:else)))))))

(event Custom|HandleRadioTowerStateChanged
  ;[0.47 0.1395 1 0.566] HandleRadioTowerStateChanged Called by OnRep_RadioTowerState()
  ; Runs on the server and clients
  (Utilities|FindRadioTowerPartsSlow
    (:All%20PartsFound
      _)
    (:NotAllPartsFound
      ;;[1 0 0] Warning: RadioTower components are null
      (Fortnite|FortShippingLog self "handleRadioTowerStateChanged: RadioTowerTerminal and or lights are null" true)))
  (Utilities|FlowControl|Sequence
    (:then_0
      (Protected|UpdateAudioCue)
      (Fortnite|FortShippingLog self (Utilities|String|Append "OnRep_RadioTowerState: New State = " (Utilities|String|EnumtoString (Variables|Transient|GetRadioTowerState))))
      (switch Utilities|FlowControl|Switch|SwitchonMMObjRadioTowerState (Variables|Transient|GetRadioTowerState)
        (:NewEnumerator0
          (State|HandleStateDeactive))
        (:NewEnumerator6
          (State|HandleStateActivating))
        (:NewEnumerator2
          (State|HandleStateActive))
        (:NewEnumerator7
          (State|HandleStateCaptured))))
    (:then_1
      (bind _radiotowerterminal (Variables|Instance|GetRadioTowerTerminal)
        (:then
          (States|RadioTower|HandleRadioTowerStateChanged _radiotowerterminal (Variables|Transient|GetRadioTowerState))
          (Default|CallOnRadioTowerStateChanged self (Variables|Transient|GetRadioTowerState)))
        (:else)))))

(fn FindRadioTowerPart_Lights
  (Variables|Instance|GetRadioTowerLights
    (:then)
    (:else
      (bind (_outactors _returnvalue) (Collision|BoxOverlapActors (Transformation|GetActorTransform) (Variables|Settings|GetLightsSearchExtent) (Utilities|Array|MakeArray "ObjectTypeQuery2") "/MidMatchRadioTowerGameplay/Gameplay/B_MMObj_RadioTowerLights.B_MMObj_RadioTowerLights_C" :!class /Script/FortniteGame.FortKismetLibrary))
      (if _returnvalue
        (bind _array_element (Utilities|Array|ForEachLoopwithBreak _outactors)
          (:LoopBody
            (bind _asb_mmobj_radio_tower_lights (Utilities|Casting|CastToB_MMObj_RadioTowerLights _array_element))
            (Variables|Instance|SetRadioTowerLights _asb_mmobj_radio_tower_lights)))))))

(fn FindRadioTowerPart_Terminal
  (Variables|Instance|GetRadioTowerTerminal
    (:then)
    (:else
      (bind (_outactors _returnvalue) (Collision|BoxOverlapActors (Transformation|GetActorTransform) (Variables|Settings|GetTerminalSearchExtent) (Utilities|Array|MakeArray "ObjectTypeQuery1") "/MidMatchRadioTowerGameplay/Gameplay/BGA_MMObj_RadioTowerTerminal.BGA_MMObj_RadioTowerTerminal_C" :!class /Script/FortniteGame.FortKismetLibrary))
      (if _returnvalue
        (bind _array_element (Utilities|Array|ForEachLoopwithBreak _outactors)
          (:LoopBody
            (bind _asbga_mmobj_radio_tower_terminal (Utilities|Casting|CastToBGA_MMObj_RadioTowerTerminal _array_element))
            (Variables|Instance|SetRadioTowerTerminal _asbga_mmobj_radio_tower_terminal)))))))

(fn GetLinkedRadioTowerTerminal
  (return (Public|GetRadioTowerTerminal)))

(fn GetRadioTowerState
  ;[0.01992 0.09956 0.1094] Return the radio tower state.
  (return (Variables|Transient|GetRadioTowerState)))

(fn GetRadioTowerStateComponent
  (bind _returnvalue (Fortnite|Athena|GameState|GetGameStateAthena))
  (bind _returnvalue_2 (Actor|GetComponentbyClass _returnvalue "/MidMatchRadioTowerGameplay/Gameplay/B_MMObj_RadioTowerGameStateComponent.B_MMObj_RadioTowerGameStateComponent_C"))
  (Utilities|IsValid _returnvalue
    (:Is%20Valid
      (Utilities|IsValid _returnvalue_2
        (:Is%20Valid
          _)))
    (:Is%20Not%20Valid
      _))
  (return _returnvalue_2))

(fn GetRadioTowerTerminal
  ;[0.03983 0.1823 0.1322] Return the radio tower terminal.
  (return (Variables|Instance|GetRadioTowerTerminal)))

(fn HandleActivationProgressUpdated
  (params (ActivationProgress :type Float%20%28double-precision%29))
  ;[0.05681 0.2148 0.276] Update the progress on the tower.
  (Networking|CanExecuteCosmeticEvents
    (:True
      (bind _radiotowerlights (Variables|Instance|GetRadioTowerLights)
        (:then
          (Class|BMMObjRadioTowerLights|SetRadioTowerActivationProgress _radiotowerlights ActivationProgress))
        (:else)))))

(fn HandleState_Activating
  ;[0.01865 0.151 0.1435] HandleState_Activating:
  ; On the client, bind to know when the activation progress gets updated.
  (Networking|SwitchHasAuthority
    (:Remote
      (bind _radiotowerlights (Variables|Instance|GetRadioTowerLights)
        (:then
          (Class|BMMObjRadioTowerLights|CosmeticTowerLightActivateStart _radiotowerlights))
        (:else
          _))
      (bind _returnvalue (Protected|GetRadioTowerStateComponent))
      (Default|BindEventtoOnRadioTowerActivationProgressUpdated _returnvalue (EventDispatchers|CreateEvent :!event HandleActivationProgressUpdated))
      (Protected|HandleActivationProgressUpdated (General|GetRadioTowerActivationProgress _returnvalue)))))

(fn HandleState_Active
  ;[0.03245 0.1771 0.08609] HandleState_Active: Activate the signal FX, update the material on the radio tower mesh, and unbind from receiving activation progress updates.
  (Networking|SwitchHasAuthority
    (:Authority
      (bind _radiotowerterminal (Variables|Instance|GetRadioTowerTerminal)
        (:then
          (Map|SetEnableMapIcon _radiotowerterminal true))
        (:else)))
    (:Remote
      (bind _radiotowerlights (Variables|Instance|GetRadioTowerLights)
        (:then
          (Class|BMMObjRadioTowerLights|CosmeticTowerLightActivateComplete _radiotowerlights))
        (:else
          _))
      (Rendering|SetVisibility (Variables|Default|GetRadar_Signal2) true false)
      (Rendering|Material|SetDefaultCustomPrimitiveDataFloat (Variables|Default|GetRadar_Signal2) 2 1.0)
      (bind _returnvalue (Protected|GetRadioTowerStateComponent))
      (Default|UnbindEventfromOnRadioTowerActivationProgressUpdated _returnvalue (EventDispatchers|CreateEvent :!event HandleActivationProgressUpdated)))))

(fn HandleState_Captured
  (Networking|CanExecuteCosmeticEvents
    (:True
      (Utilities|FlowControl|Sequence
        (:then_0
          (bind _radiotowerlights (Variables|Instance|GetRadioTowerLights)
            (:then
              (Class|BMMObjRadioTowerLights|CosmeticTowerLightDisable _radiotowerlights))
            (:else)))
        (:then_1
          (Rendering|SetVisibility (Variables|Default|GetRadar_Signal2) false false))))))

(fn HandleState_Deactive
  ;[0.04542 0.1198 0.1302] HandleState_Deactive: Deactivate all the radio tower behavior.
  (Networking|SwitchHasAuthority
    (:Remote
      (bind _radiotowerlights (Variables|Instance|GetRadioTowerLights)
        (:then
          (Class|BMMObjRadioTowerLights|CosmeticTowerLightDisable _radiotowerlights))
        (:else)))))

(fn OnRep_RadioTowerState
  ;[0.04921 0.232 0.3021] OnRep_RadioTowerState: Handle the various state changes of the radio tower.
  ; Called on both server and client.
  (CallFunction|HandleRadioTowerStateChanged))

(fn OnTowerPhaseAdvance
  (CallFunction|UpdateRadioTowerState))

(fn SetRadioTowerState
  (params (NewState :type MMObjRadioTowerState%20Enum :default "NewEnumerator5"))
  ;[0.03026 0.1233 0.1354] SetRadioTowerState: Update the state of the radio tower.
  (Networking|SwitchHasAuthority
    (:Authority
      (if (and (Utilities|Enum|NotEqual(Enum) (Variables|Transient|GetRadioTowerState) "NewEnumerator7") (Utilities|Enum|NotEqual(Enum) (Variables|Transient|GetRadioTowerState) NewState))
        (Variables|Transient|SetRadioTowerState NewState)
        (CallFunction|MulticastHandleRadioTowerStateUpdate NewState)
        (Networking|FlushNetDormancy)
        ;[0.03026 0.1233 0.1354] Used for Analytics
        (if (Utilities|Enum|Equal(Enum) (Variables|Transient|GetRadioTowerState) "NewEnumerator7")
          (bind _returnvalue (Protected|GetRadioTowerStateComponent))
          (General|HandleCapturedRadioTowerState _returnvalue self "NewEnumerator7"))))))

(fn SpawnGuards
  (bind _radiotowerterminal (Variables|Instance|GetRadioTowerTerminal))
  (if (not (Variables|Transient|GetSpawnedGuards))
    (Utilities|IsValid _radiotowerterminal
      (:Is%20Valid
        (Variables|Transient|SetSpawnedGuards true)
        (Guards|SpawnRadioTowerGuards _radiotowerterminal)))))

(fn SyncLightsTransform
  (bind _radiotowerlights (Variables|Instance|GetRadioTowerLights)
    (:then
      (Transformation|SetActorTransform _radiotowerlights (Transformation|GetActorTransform) false false))
    (:else)))

(fn TurnOffRadioSignal
  (Rendering|SetVisibility (Variables|Default|GetRadar_Signal2) false false))

(fn UpdateAudioCue
  (bind _radiotowerstate (CallFunction|GetRadioTowerState))
  (bind _stateupdateloopsoundmap (Variables|Settings|GetStateUpdateLoopSoundMap))
  (bind _value (Utilities|Map|Find _stateupdateloopsoundmap _radiotowerstate))
  (bind _ambientloopaudioclosecomponent (Variables|Default|GetAmbientLoopAudioCloseComponent))
  (bind _ambientloopaudiocomponent (Variables|Default|GetAmbientLoopAudioComponent))
  ;[0.003723 0.1727 0.2188] Update the audio loop component based on the state.
  (Networking|CanExecuteCosmeticEvents
    (:True
      (Utilities|IsValid _value
        (:Is%20Valid
          (Audio|Components|Audio|SetSound _ambientloopaudiocomponent _value)
          (Audio|Components|Audio|Play _ambientloopaudiocomponent 0.0)
          (Audio|Components|Audio|Play _ambientloopaudioclosecomponent 0.0))
        (:Is%20Not%20Valid
          (Audio|Components|Audio|FadeOut (Variables|Default|GetAmbientLoopAudioComponent) 0.5 0.0 "Linear" :!class /Script/Engine.AudioComponent)
          (Audio|Components|Audio|FadeOut _ambientloopaudioclosecomponent 0.5 0.0 "Linear" :!class /Script/Engine.AudioComponent))))))

(fn UpdateRadioTowerState
  (bind _returnvalue (Protected|GetRadioTowerStateComponent))
  (bind _radiotowerphasestate (General|GetRadioTowerState _returnvalue))
  (Public|SetRadioTowerState _radiotowerphasestate))

(fn ConstructionScript
  ;[0.05024 0.1237 0.1615] Construction: Sync up the location/rotation of the radio tower lights and the radio tower.
  (|CallParentFunction)
  (CallFunction|SyncLightsTransform))
