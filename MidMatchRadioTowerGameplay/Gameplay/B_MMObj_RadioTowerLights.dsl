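; Cosmetic lights Blueprint for the mid-match radio tower objective, showing activation start, progress, completion, red locked and disabled states.
; Events drive the lights mesh through custom primitive data: slot 0 enables, 1 carries progress, 2 marks completion, 3 turns red.
; SetRadioTowerActivationProgress stores progress, then remaps it through a curve, if valid, into slot 1.
; Locked reuses the completed state before setting red; Disable clears only the enable slot.
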(event Custom|CosmeticTowerLightActivateStart
  ;[0.01686 0.2248 0.2812] CosmeticTowerLightActivateStart: The tower light activation is starting/progressing.
  (Rendering|Material|SetDefaultCustomPrimitiveDataFloat (Variables|Default|GetLightsStaticMeshComponent) 0 1.0)
  (Rendering|Material|SetDefaultCustomPrimitiveDataFloat (Variables|Default|GetLightsStaticMeshComponent) 1 0.0)
  (Rendering|Material|SetDefaultCustomPrimitiveDataFloat (Variables|Default|GetLightsStaticMeshComponent) 2 0.0)
  (Rendering|Material|SetDefaultCustomPrimitiveDataFloat (Variables|Default|GetLightsStaticMeshComponent) 3 0.0))

(event Custom|CosmeticTowerLightActivateComplete
  ;[0.02355 0.2083 0.198] CosmeticTowerLightActivateComplete: The tower light activation is complete.
  (Rendering|Material|SetDefaultCustomPrimitiveDataFloat (Variables|Default|GetLightsStaticMeshComponent) 0 1.0)
  (Rendering|Material|SetDefaultCustomPrimitiveDataFloat (Variables|Default|GetLightsStaticMeshComponent) 1 1.0)
  (Rendering|Material|SetDefaultCustomPrimitiveDataFloat (Variables|Default|GetLightsStaticMeshComponent) 2 1.0))

(event Custom|CosmeticTowerLightLocked
  ;[0.5985 0.177 1] CosmeticTowerLightLocked: Set the lights on the tower to red.
  (CallFunction|CosmeticTowerLightActivateComplete)
  (Rendering|Material|SetDefaultCustomPrimitiveDataFloat (Variables|Default|GetLightsStaticMeshComponent) 3 1.0))

(event Custom|CosmeticTowerLightDisable
  ;[0.0548 0.276 0.1732] CosmeticTowerLightDisable: Disable the lights.
  (Rendering|Material|SetDefaultCustomPrimitiveDataFloat (Variables|Default|GetLightsStaticMeshComponent) 0 0.0))

(event Custom|CosmeticTowerLightActivateProgress
  ;[0.01833 0.1341 0.1354] CosmeticTowerLightActivateProgress: Update the progress (0-1) scale.
  (bind _radiotoweractivationprogresscurve (Variables|Default|GetRadioTowerActivationProgressCurve)
    (:then
      (Rendering|Material|SetDefaultCustomPrimitiveDataFloat (Variables|Default|GetLightsStaticMeshComponent) 1 (Math|Curves|GetFloatValue _radiotoweractivationprogresscurve (Variables|Default|GetRadioTowerActivationProgress))))
    (:else)))

(fn SetRadioTowerActivationProgress
  (params (Progress :type Float%20%28double-precision%29))
  ;[0.02043 0.1406 0.1364] Update the radio tower activation progress.
  (Variables|Default|SetRadioTowerActivationProgress Progress)
  (CallFunction|CosmeticTowerLightActivateProgress))

(fn ConstructionScript)
