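; Blueprint for the screen of a mid-match objective radio tower terminal, showing state, text and errors through a world widget.
; BeginPlay caches the terminal widget, and custom events forward text, error and activate-animation calls to it or toggle screen visibility.
; State changes look up inner and outer glow colors in maps, then a timeline lerps the widget's dynamic material to them.
; Error mode overrides these colors, and clearing it restores the latest state's colors.
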(event EventBeginPlay
  ;[0.3181 0.7188 0.1776] BeginPlay: Cache the terminal screen widget and disable the widget tick.
  (bind _aswbp_mmobj_radio_tower_terminal_screen (Utilities|Casting|CastToWBP_MMObj_RadioTowerTerminalScreen (UserInterface|GetUserWidgetObject (Variables|Default|GetFortScreenWidget))))
  (Variables|Transient|SetTerminalScreenWidget _aswbp_mmobj_radio_tower_terminal_screen))

(event Custom|UpdateScreen
  (params (TerminalState :type MMObjRadioTowerState%20Enum :default "NewEnumerator5") (InText :type Text%20%28by%20ref%29))
  ;[0.08433 0.03634 0.2292] UpdateScreen: Called to update the state/text of the screen.
  (bind _output_get (Variables|Transient|SetLatestTerminalState TerminalState))
  (CallFunction|RefreshScreenVisual (Utilities|Map|Find (Variables|Settings|GetTerminalStateToInnerColor) _output_get) (Utilities|Map|Find (Variables|Settings|GetTerminalStateToOuterColor) _output_get))
  (bind _terminalscreenwidget (Variables|Transient|GetTerminalScreenWidget)
    (:then
      (Class|WBPMMObjRadioTowerTerminalScreen|UpdatePrimaryText _terminalscreenwidget InText))
    (:else)))

(event Custom|UpdatePrimaryText
  (params (InText :type Text%20%28by%20ref%29))
  ;[0.03103 0.2357 0.2708] UpdateText: Update the primary text of the widget.
  (Class|WBPMMObjRadioTowerTerminalScreen|UpdatePrimaryText (Variables|Transient|GetTerminalScreenWidget) InText))

(event Custom|SetEnableScreen
  (params (bEnable :type Boolean))
  ;[0.05065 0.1971 0.2031] SetEnableScreen: Called to enable or disable the screen.
  (Development|SetHiddeninGame (Variables|Default|GetFortScreenWidget) (not bEnable) false))

(event Custom|RefreshScreenVisual
  (params (InnerGlowColor :type Linear%20Color%20Structure) (OuterGlowColor :type Linear%20Color%20Structure))
  ;[0.894 0.2379 1] RefreshScreenVisual: Update the screen color based on the terminal state.
  (bind (_direction _alpha) (Timeline|Timeline PlayFromStart)
    (:Update
      (bind _fortscreenwidget (Variables|Default|GetFortScreenWidget)
        (:then
          (Utilities|IsValid (UserInterface|GetMaterialInstance _fortscreenwidget)
            (:Is%20Valid
              (Rendering|Material|SetVectorParameterValue (UserInterface|GetMaterialInstance _fortscreenwidget) "Glow Color Inner" (Math|Color|Lerp(LinearColor) (Variables|Default|GetCurrentInnerColor) InnerGlowColor _alpha) :!class /Script/Engine.MaterialInstanceDynamic)
              (Rendering|Material|SetVectorParameterValue (UserInterface|GetMaterialInstance _fortscreenwidget) "Glow Color Outer" (Math|Color|Lerp(LinearColor) (Variables|Default|GetCurrentOuterColor) OuterGlowColor _alpha) :!class /Script/Engine.MaterialInstanceDynamic))))
        (:else)))
    (:Finished
      (Variables|Default|SetCurrentInnerColor InnerGlowColor)
      (Variables|Default|SetCurrentOuterColor OuterGlowColor))))

(event Custom|SetEnableErrorScreen
  (params (bEnable :type Boolean))
  ;[0.02711 0.2061 0.224] SetEnableErrorScreen: Update the error state on the screen widget and update the material on the widget as well.
  (bind _terminalscreenwidget (Variables|Transient|GetTerminalScreenWidget)
    (:then
      (Class|WBPMMObjRadioTowerTerminalScreen|SetEnableErrorScreen _terminalscreenwidget bEnable)
      (if bEnable
        (CallFunction|RefreshScreenVisual (Variables|Settings|GetErrorInnerColor) (Variables|Settings|GetErrorOuterColor)) ; Error
        (else
          (CallFunction|RefreshScreenVisual (Utilities|Map|Find (Variables|Settings|GetTerminalStateToInnerColor) (Variables|Transient|GetLatestTerminalState)) (Utilities|Map|Find (Variables|Settings|GetTerminalStateToOuterColor) (Variables|Transient|GetLatestTerminalState))))))
    (:else)))

(event Custom|SetEnableSecondaryText
  (params (bEnable :type Boolean))
  ;[0.1421 0.03745 0.1615] SetEnableSecondaryText: Enable or disable the secondary text value.
  (bind _terminalscreenwidget (Variables|Transient|GetTerminalScreenWidget)
    (:then
      (Class|WBPMMObjRadioTowerTerminalScreen|SetEnableSecondaryText _terminalscreenwidget bEnable))
    (:else)))

(event Custom|UpdateSecondaryText
  (params (InText :type Text))
  ;[0.07385 0.05553 0.2865] UpdateSecondaryText: Update the secondary text value.
  (bind _terminalscreenwidget (Variables|Transient|GetTerminalScreenWidget)
    (:then
      (Class|WBPMMObjRadioTowerTerminalScreen|UpdateSecondaryText _terminalscreenwidget InText))
    (:else)))

(event Custom|PlayActivateAnimation
  ;[0.05828 0.1298 0.1823] PlayActivateAnimation: Play the activate animation.
  ; This is the animation that plays when the screen goes from blank to showing the progress percentage.
  (bind _terminalscreenwidget (Variables|Transient|GetTerminalScreenWidget)
    (:then
      (Class|WBPMMObjRadioTowerTerminalScreen|PlayActivateAnimation _terminalscreenwidget))
    (:else)))

(fn ConstructionScript)
